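# Physics asset for the female medium Winter Onesie body, covering dynamic fur and eye bones on both slippers.
# Each ball bone carries a capsule body and anchors four fur capsules and two zero-length eye capsules on its side.
# Fur constraints use tight swing and twist limits for a small wobble.
# Eye constraints lock both swing axes and set a wide twist limit, so the eyes only rotate about the twist axis.
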PhysicsAsset /BRCosmetics/Characters/Player/Female/Medium/Bodies/F_MED_WinterOnesie/Meshes/F_MED_Winter_onesie_Physics
bodies {
 ball_l [{shapeType=Capsule, length=2, radius=2}]
 ball_r [{shapeType=Capsule, length=2, radius=2}]
 dyn_eye_01_l [{shapeType=Capsule, length=0, radius=2}]
 dyn_eye_01_r [{shapeType=Capsule, length=0, radius=2}]
 dyn_eye_02_l [{shapeType=Capsule, length=0, radius=2}]
 dyn_eye_02_r [{shapeType=Capsule, length=0, radius=2}]
 dyn_slipper_fur_01_l [{shapeType=Capsule, length=2, radius=2, center=<0 1 0>, rotation=<rot 0 0 90>}]
 dyn_slipper_fur_01_r [{shapeType=Capsule, length=2, radius=2, center=<0 1 0>, rotation=<rot 0 0 90>}]
 dyn_slipper_fur_02_l [{shapeType=Capsule, length=2, radius=2, center=<0 1 0>, rotation=<rot 0 0 90>}]
 dyn_slipper_fur_02_r [{shapeType=Capsule, length=2, radius=2, center=<0 1 0>, rotation=<rot 0 0 90>}]
 dyn_slipper_fur_03_l [{shapeType=Capsule, length=2, radius=2, center=<0 1 0>, rotation=<rot 0 0 90>}]
 dyn_slipper_fur_03_r [{shapeType=Capsule, length=2, radius=2, center=<0 1 0>, rotation=<rot 0 0 90>}]
 dyn_slipper_fur_04_l [{shapeType=Capsule, length=2, radius=2, center=<0 1 0>, rotation=<rot 0 0 90>}]
 dyn_slipper_fur_04_r [{shapeType=Capsule, length=2, radius=2, center=<0 1 0>, rotation=<rot 0 0 90>}]
}
constraints [
 {bone1Name=dyn_slipper_fur_04_l, bone2Name=ball_l, swing1LimitDegrees=20, swing2LimitDegrees=20, twistLimitDegrees=20}
 {bone1Name=dyn_slipper_fur_03_l, bone2Name=ball_l, swing1LimitDegrees=20, swing2LimitDegrees=20, twistLimitDegrees=20}
 {bone1Name=dyn_slipper_fur_02_l, bone2Name=ball_l, swing1LimitDegrees=20, swing2LimitDegrees=20, twistLimitDegrees=20}
 {bone1Name=dyn_slipper_fur_01_l, bone2Name=ball_l, swing1LimitDegrees=20, swing2LimitDegrees=20, twistLimitDegrees=20}
 {
  bone1Name=dyn_eye_01_l
  bone2Name=ball_l
  swing1LimitDegrees=20
  swing1Motion=Locked
  swing2LimitDegrees=20
  swing2Motion=Locked
  twistLimitDegrees=180
 }
 {
  bone1Name=dyn_eye_02_l
  bone2Name=ball_l
  swing1LimitDegrees=20
  swing1Motion=Locked
  swing2LimitDegrees=20
  swing2Motion=Locked
  twistLimitDegrees=180
 }
 {bone1Name=dyn_slipper_fur_01_r, bone2Name=ball_r, swing1LimitDegrees=20, swing2LimitDegrees=20, twistLimitDegrees=20}
 {bone1Name=dyn_slipper_fur_02_r, bone2Name=ball_r, swing1LimitDegrees=20, swing2LimitDegrees=20, twistLimitDegrees=20}
 {bone1Name=dyn_slipper_fur_04_r, bone2Name=ball_r, swing1LimitDegrees=20, swing2LimitDegrees=20, twistLimitDegrees=20}
 {bone1Name=dyn_slipper_fur_03_r, bone2Name=ball_r, swing1LimitDegrees=20, swing2LimitDegrees=20, twistLimitDegrees=20}
 {
  bone1Name=dyn_eye_02_r
  bone2Name=ball_r
  swing1LimitDegrees=20
  swing1Motion=Locked
  swing2LimitDegrees=20
  swing2Motion=Locked
  twistLimitDegrees=180
 }
 {
  bone1Name=dyn_eye_01_r
  bone2Name=ball_r
  swing1LimitDegrees=20
  swing1Motion=Locked
  swing2LimitDegrees=20
  swing2Motion=Locked
  twistLimitDegrees=180
 }
]
