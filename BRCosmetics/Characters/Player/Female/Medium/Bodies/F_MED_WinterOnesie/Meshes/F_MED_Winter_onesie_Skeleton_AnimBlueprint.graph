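# Post-process style AnimBlueprint for the female medium Winter Onesie body cosmetic, adding secondary motion to the mesh.
# It copies the pose from the attached parent mesh, then chains three Constraint nodes that place dyn_shell_01 to 03 by weighted blends of left clavicle, upper-arm, twist and upper-spine bones.
# A RigidBody node then simulates the onesie physics asset in root-bone space with amplified component velocity and acceleration.
# The event graph contains no logic.
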AnimBlueprint /BRCosmetics/Characters/Player/Female/Medium/Bodies/F_MED_WinterOnesie/Meshes/F_MED_Winter_onesie_Skeleton_AnimBlueprint
target_skeleton=&/F_MED_Winter_onesie_Skeleton
graphs [
 AnimGraph {nodes [
   n1 = "Constraint - dyn_shell_03 (PPPPPP)"!(
    ComponentPose=Local To Component(LocalPose=Copy Pose From Mesh() {bUseAttachedParent=true})
    Alpha=1
    ConstraintWeights_0=0.2162
    ConstraintWeights_1:float=0.0005
    ConstraintWeights_2:float=0.495
    ConstraintWeights_3:float=0.2323
    ConstraintWeights_4:float=0.0555
    ConstraintWeights_5:float=0.0005
   ) -> (Pose) {
    alphaBoolBlend {blendOption=Linear}
    boneToModify {boneName=dyn_shell_03}
    constraintSetup [
     {targetBone {boneName=clavicle_l}}
     {targetBone {boneName=upperarm_twist_02_l}}
     {targetBone {boneName=spine_04}}
     {targetBone {boneName=spine_05}}
     {targetBone {boneName=upperarm_twist_01_l}}
     {targetBone {boneName=upperarm_l}}
    ]
    constraintWeights [1,1,1,1,1,1]
   }
   n2 = "Constraint - dyn_shell_02 (PPPPPP)"!(
    ComponentPose=@n1
    Alpha=1
    ConstraintWeights_0=0.4986
    ConstraintWeights_1:float=0.0002
    ConstraintWeights_2:float=0.2307
    ConstraintWeights_3:float=0.2156
    ConstraintWeights_4:float=0.0548
    ConstraintWeights_5:float=0.0002
   ) -> (Pose) {
    alphaBoolBlend {blendOption=Linear}
    boneToModify {boneName=dyn_shell_02}
    constraintSetup [
     {targetBone {boneName=clavicle_l}}
     {targetBone {boneName=upperarm_twist_02_l}}
     {targetBone {boneName=spine_04}}
     {targetBone {boneName=spine_05}}
     {targetBone {boneName=upperarm_twist_01_l}}
     {targetBone {boneName=upperarm_l}}
    ]
    constraintWeights [1,1,1,1,1,1]
   }
   n3 = "Constraint - dyn_shell_01 (PPPPPP)"!(
    ComponentPose=@n2
    Alpha=1
    ConstraintWeights_0=0.8048
    ConstraintWeights_1:float=0.0002
    ConstraintWeights_2:float=0.0855
    ConstraintWeights_3:float=0.0638
    ConstraintWeights_4:float=0.0456
    ConstraintWeights_5:float=0.0001
   ) -> (Pose) {
    alphaBoolBlend {blendOption=Linear}
    boneToModify {boneName=dyn_shell_01}
    constraintSetup [
     {targetBone {boneName=clavicle_l}}
     {targetBone {boneName=upperarm_twist_02_l}}
     {targetBone {boneName=spine_04}}
     {targetBone {boneName=spine_05}}
     {targetBone {boneName=upperarm_twist_01_l}}
     {targetBone {boneName=upperarm_l}}
    ]
    constraintWeights [1,1,1,1,1,1]
   }
   n4 = RigidBody(ComponentPose=@n3) {
    bClampLinearTranslationLimitToRefPose=true
    overridePhysicsAsset=&/F_MED_Winter_onesie_Physics
    simulationSpace=BaseBoneSpace
    alphaBoolBlend {blendOption=Linear}
    baseBoneRef {boneName=root}
    componentLinearAccScale=<10 10 10>
    componentLinearVelScale=<10 10 10>
   }
   Output Pose(Result=Component To Local(ComponentPose=@n4)) {layerGroup=None}
  ]}
 EventGraph {nodes [
   TryGetPawnOwner()
   Event BlueprintUpdateAnimation()
  ]}
]
variables []
