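# Physics asset for the face accessory mesh of the female medium Winter Onesie cosmetic, simulating its dynamic bangs.
# A sphere-like zero-length capsule on the head and a capsule on the dyn_bangs_01 bone form the two bodies.
# One constraint ties the bangs to the head, allowing limited swing on both axes with a zero twist limit.
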PhysicsAsset /BRCosmetics/Characters/Player/Female/Medium/Bodies/F_MED_WinterOnesie/Meshes/Parts/F_MED_Winter_onesie_FaceAcc_Physics
bodies {
 dyn_bangs_01 [{shapeType=Capsule, length=4.8, radius=1.7, center=<2.5 0 0>, rotation=<rot 90 0 0>}]
 head [{shapeType=Capsule, length=0, radius=5}]
}
constraints [{bone1Name=dyn_bangs_01, bone2Name=head, swing1LimitDegrees=40, swing2LimitDegrees=40, twistLimitDegrees=0}]
